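# UI-domain material, named for a video player widget, built around the WidgetMediaPlayer_Video texture.
# It is unlit and translucent, and its only node is a texture sample of that video texture, with no parameters or other processing.
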Material /Game/Movies/VideoPlayerWidgetMaterial_Unique
graph {
 properties {blendMode=BLEND_Translucent, materialDomain=MD_UI, pixelDepthOffsetMode=PDOM_Legacy, shadingModel=MSM_Unlit, usedShadingModels=Unlit}
 nodes [TextureSample() {texture=&/WidgetMediaPlayer_Video}]
}
