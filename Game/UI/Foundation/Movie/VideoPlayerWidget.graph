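# VideoPlayerWidget is a Widget Blueprint built on the native FortVideoPlayerWidget class that displays video with subtitles in the UI.
# An Overlay root holds the inherited video image, which fills the space and draws VideoPlayerWidgetMaterial at 1920x1080.
# A hit-test-invisible SubtitleDisplay sits bottom-center above it, styled by the shared SubtitleOptions asset with wrapped text.
# The event graph is empty, so the Blueprint supplies only layout and styling for the parent class.
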WidgetBlueprint /Game/UI/Foundation/Movie/VideoPlayerWidget
graphs [EventGraph {links [], nodes []}]
variables []
widgets {
 info {inheritedWidgetCount=2, parentClass=/Script/FortniteUI.FortVideoPlayerWidget, rootWidgetClass=&Overlay, widgetCount=3}
 widgets [
  Overlay_0 {widgetClassPath=&Overlay, widget=Overlay_0}
  Image_VideoTexture {
   widgetClassPath=&Image
   bInherited=true
   bIsVariable=true
   parent=Overlay_0
   widget=Image_VideoTexture
   properties {brush {
     bIsDynamicallyLoaded=false
     drawAs=Image
     imageType=NoImage
     mirroring=NoMirror
     resourceName=None
     resourceObject=&/VideoPlayerWidgetMaterial
     tiling=NoTile
     imageSize=<1920 1080>
     margin=<ltrb 0 0 0 0>
     outlineSettings {
      bUseBrushTransparency=false
      roundingType=HalfHeightRadius
      width=0
      color {colorUseRule=UseColor_Specified, specifiedColor=<rgba 0 0 0 0>}
      cornerRadii=<0 0 0 0>
     }
     tintColor {colorUseRule=UseColor_Specified, specifiedColor=<rgba 1 1 1 1>}
     uVRegion {bIsValid=false, max=<0 0>, min=<0 0>}
    }}
   slot {class=OverlaySlot, properties {horizontalAlignment=HAlign_Fill, verticalAlignment=VAlign_Fill}}
  }
  Subtitles {
   widgetClassPath=/Script/SubtitlesWidgets.SubtitleDisplay
   bInherited=true
   bIsVariable=true
   parent=Overlay_0
   widget=Subtitles
   properties {
    options=/Game/UI/SubtitleOptions
    previewText=SEE-Bot: SEE-Bot SEES THE INSIDE OF THE BAG. THERE IS SO MUCH BAG TO SEE.
    visibility=HitTestInvisible
    wrapTextAt=1170
   }
   slot {class=OverlaySlot, properties {horizontalAlignment=HAlign_Center, verticalAlignment=VAlign_Bottom, padding=<ltrb 0 0 0 100>}}
  }
 ]
}
