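# Activatable fullscreen movie panel, derived from FortActivatableVideoPanel, that shows a video with subtitles and a Skip button.
# A black safe-zone border holds an overlay with the scale-boxed inherited video player and a bottom-right Skip button bound to GenericBack.
# The overlay also holds an initially hidden black backdrop and centered throbber.
# On Construct, it assigns VideoPlayerWidgetMaterial_Unique as the player's media material.
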WidgetBlueprint /Game/UI/Foundation/Movie/ActivatableMovieWidget
graphs [EventGraph {nodes [
   Event Construct()
   > SetMediaMaterial!(
    execute
    self:Fort Video Player Widget Object Reference=Get VideoPlayerWidget() -> (VideoPlayerWidget:Fort Video Player Widget Object Reference)
    VideoMaterial:Material Interface Object Reference=/Game/Movies/VideoPlayerWidgetMaterial_Unique
   ) -> (then)
  ]}]
variables []
widgets {
 info {inheritedWidgetCount=4, parentClass=/Script/FortniteUI.FortActivatableVideoPanel, rootWidgetClass=&SafeZone, widgetCount=9}
 widgets [
  MainSafeZone {widgetClassPath=&SafeZone, bIsVariable=true, widget=MainSafeZone, properties {padBottom=false, padTop=false}}
  SafeZonePadding {
   widgetClassPath=&Border
   parent=MainSafeZone
   widget=SafeZonePadding
   properties {visibility=SelfHitTestInvisible, brushColor=<rgba 0 0 0 1>, padding=<ltrb 0 0 0 0>}
   slot {class=SafeZoneSlot}
  }
  Overlay_0 {
   widgetClassPath=&Overlay
   parent=SafeZonePadding
   widget=Overlay_0
   properties {visibility=Visible}
   slot {class=BorderSlot, properties {padding=<ltrb 0 0 0 0>}}
  }
  ScaleBox_0 {
   widgetClassPath=&ScaleBox
   parent=Overlay_0
   widget=ScaleBox_0
   slot {class=OverlaySlot, properties {horizontalAlignment=HAlign_Fill, verticalAlignment=VAlign_Fill}}
  }
  VideoPlayerWidget {
   widgetClassPath=&/VideoPlayerWidget
   bInherited=true
   bIsVariable=true
   parent=ScaleBox_0
   widget=VideoPlayerWidget
   properties {bShowSubtitles=true}
   slot {class=ScaleBoxSlot}
  }
  SafeZone_0 {
   widgetClassPath=&SafeZone
   bIsVariable=true
   parent=Overlay_0
   widget=SafeZone_0
   properties {padLeft=false, padTop=false}
   slot {class=OverlaySlot, properties {horizontalAlignment=HAlign_Right, verticalAlignment=VAlign_Bottom, padding=<ltrb 0 0 20 20>}}
  }
  Button_Skip {
   widgetClassPath=/FortUILibrary/UIKit/Buttons/WBP_UIKit_Button_Quiet
   bInherited=true
   bIsVariable=true
   parent=SafeZone_0
   widget=Button_Skip
   properties {
    bHideInputActionWithKeyboard=true
    text=Skip
    triggeringInputAction {dataTable=&InputActionDataTable, rowName=GenericBack}
   }
   slot {class=SafeZoneSlot}
  }
  Image_ThrobberBackground {
   widgetClassPath=&Image
   bInherited=true
   bIsVariable=true
   parent=Overlay_0
   widget=Image_ThrobberBackground
   properties {
    visibility=Hidden
    brush {
     bIsDynamicallyLoaded=false
     drawAs=Box
     imageType=NoImage
     mirroring=NoMirror
     resourceName=None
     resourceObject=null
     tiling=NoTile
     imageSize=<1 1>
     margin=<ltrb 0 0 0 0>
     outlineSettings {
      bUseBrushTransparency=false
      roundingType=HalfHeightRadius
      width=0
      color {colorUseRule=UseColor_Specified, specifiedColor=<rgba 0 0 0 0>}
      cornerRadii=<0 0 0 0>
     }
     tintColor {colorUseRule=UseColor_Specified, specifiedColor=<rgba 1 1 1 1>}
     uVRegion {bIsValid=false, max=<0 0>, min=<0 0>}
    }
    colorAndOpacity=<rgba 0 0 0 1>
   }
   slot {class=OverlaySlot, properties {horizontalAlignment=HAlign_Fill, verticalAlignment=VAlign_Fill}}
  }
  Image_Throbber {
   widgetClassPath=&Image
   bInherited=true
   bIsVariable=true
   parent=Overlay_0
   widget=Image_Throbber
   properties {
    visibility=Hidden
    brush {
     bIsDynamicallyLoaded=false
     drawAs=Image
     imageType=NoImage
     mirroring=NoMirror
     resourceName=None
     resourceObject=/FortUILibrary/UIKit/Throbber/MI_UIKit_Monochrome_Throbber
     tiling=NoTile
     imageSize=<64 64>
     margin=<ltrb 0 0 0 0>
     outlineSettings {
      bUseBrushTransparency=false
      roundingType=HalfHeightRadius
      width=0
      color {colorUseRule=UseColor_Specified, specifiedColor=<rgba 0 0 0 0>}
      cornerRadii=<0 0 0 0>
     }
     tintColor {colorUseRule=UseColor_Specified, specifiedColor=<rgba 1 1 1 1>}
     uVRegion {bIsValid=false, max=<0 0>, min=<0 0>}
    }
   }
   slot {class=OverlaySlot, properties {horizontalAlignment=HAlign_Center, verticalAlignment=VAlign_Center}}
  }
 ]
}
