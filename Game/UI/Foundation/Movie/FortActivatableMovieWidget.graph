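# Fullscreen movie player widget on FortActivatableMovieWidget, with scaled video, subtitles, an initially collapsed Skip button, throbber and color overlay.
# Parent fade events play reversible color-overlay or whole-widget animations, then report completion back to the parent.
# On activation it binds media start; for the season Combo (FTUE) video, it schedules a pre-end transition sound that persists across level loads, plus a leftover debug print.
# Deactivation clears these bindings.
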WidgetBlueprint /Game/UI/Foundation/Movie/FortActivatableMovieWidget
animations [
 ColorFade {
  binding_count=1
  length_seconds=1
  object=&self:ColorFade
  sequence {
   bindings [Image_Color {
     id=#1
     possessed_class=Image
     tracks [{
       type=MovieSceneColorTrack
       display_name=Color and Opacity
       sections [{
         type=MovieSceneColorSection
         end_frame=60
         start_frame=0
         channels [
          Color.R {type=MovieSceneScriptingFloatChannel}
          Color.G {type=MovieSceneScriptingFloatChannel}
          Color.B {type=MovieSceneScriptingFloatChannel}
          Color.A {type=MovieSceneScriptingFloatChannel, keys [[0,1],[60,0]]}
         ]
        }]
      }]
    }]
   display_rate {numerator=60}
  }
 }
 WholeFade {
  binding_count=1
  length_seconds=1
  object=&self:WholeFade
  sequence {bindings [FortActivatableMovieWidget_C_9 {
     id=#2
     possessed_class=FortActivatableMovieWidget_C
     tracks [{
       type=MovieSceneColorTrack
       display_name=Color and Opacity
       sections [{
         type=MovieSceneColorSection
         end_frame=20
         start_frame=0
         channels [
          Color.R {type=MovieSceneScriptingFloatChannel}
          Color.G {type=MovieSceneScriptingFloatChannel}
          Color.B {type=MovieSceneScriptingFloatChannel}
          Color.A {type=MovieSceneScriptingFloatChannel, keys [[0,1],[20,0]]}
         ]
        }]
      }]
    }]}
 }
]
graphs [
 EventGraph {nodes [
   n1 = PreEnded() {is=title K2Node_CustomEvent}
   > SpawnSound2D!(
    execute
    Sound:Sound Base Object Reference=Get TransitionSound() -> (TransitionSound:Sound Base Object Reference)
    VolumeMultiplier:float=1
    PitchMultiplier:float=1
    StartTime:float=0
    ConcurrencySettings:Sound Concurrency Object Reference
    bPersistAcrossLevelTransition:bool=true
    bAutoDestroy:bool=true
   ) -> (then, ReturnValue:Audio Component Object Reference)
   > PrintString!(execute, InString="Doing Work!", bPrintToScreen, bPrintToLog, TextColor, Duration, Key) -> (then)
   n2 = MediaStarted() {is=title K2Node_CustomEvent}
   > Is FTUEVideo!(Check:Exec) -> (True:Exec, False:Exec) {is=title K2Node_MacroInstance}
   > SetMediaPreEndTimeSeconds!(
    execute
    self:Fort Activatable Movie Widget Object Reference
    InSeconds:float=Get SecondsPriorToTransSound() -> (SecondsPriorToTransSound:float)
   ) -> (then)
   > Bind Event to Media Pre End Event(self:Fort Activatable Movie Widget Object Reference, Delegate=@n1) {is=title K2Node_AddDelegate}
   n3 = Event StartWholeFade!() -> (OutputDelegate, then, bReverse:bool) {is=title K2Node_Event}
   > n4 = "Play Animation with Finished event (legacy)"!(
    execute:Exec
    Widget:User Widget Object Reference=Self-Reference()
    InAnimation:Widget Animation Object Reference=Get WholeFade() -> (WholeFade:Widget Animation Object Reference)
    StartAtTime:float=0
    NumLoopsToPlay:int=1
    PlayMode:EUMGSequencePlayMode Enum=Select(Option 0:EUMGSequencePlayMode Enum, Option 1:EUMGSequencePlayMode Enum=Reverse, Index=@n3.bReverse) -> (ReturnValue:EUMGSequencePlayMode Enum)
    PlaybackSpeed:float=1
   ) -> (then:Exec, Result:UMGSequence Player Object Reference, Finished:Exec) {is=title K2Node_PlayAnimation}
   WholeFadeComplete(execute=@n4.Finished, self:Fort Activatable Movie Widget Object Reference)
   Event On Activated()
   > Bind Event to Media Started Event(self:Fort Activatable Movie Widget Object Reference, Delegate=@n2) {is=title K2Node_AddDelegate}
   Event On Deactivated()
   > Unbind all Events from Media Started Event!(execute:Exec, self:Fort Activatable Movie Widget Object Reference) -> (then:Exec) {is=title K2Node_ClearDelegate}
   > Is FTUEVideo()
   > Unbind all Events from Media Pre End Event!(execute:Exec, self:Fort Activatable Movie Widget Object Reference) -> (then:Exec) {is=title K2Node_ClearDelegate}
   n5 = Event StartFadeFromColor!() -> (OutputDelegate, then, bEnd:bool, Reason:UFortActivatablePrePostRollStates Enum=None, bReverse:bool) {is=title K2Node_Event}
   > n6 = "Play Animation with Finished event (legacy)"(
    Widget=Self-Reference()
    InAnimation=Get ColorFade() -> (ColorFade:Widget Animation Object Reference)
    StartAtTime=0
    NumLoopsToPlay=1
    PlayMode=Select(Option 1=Reverse, Index=@n5.bReverse)
    PlaybackSpeed=1
   )
   FadeInFromColorComplete!(execute=@n6.Finished, self:Fort Activatable Movie Widget Object Reference, Reason:UFortActivatablePrePostRollStates Enum=@n5.Reason) -> (then)
   Event PreConstruct()
   Event Construct()
   Event Tick()
  ]}
 IsFTUEVideo {nodes [
   Inputs!() -> (Check:Exec)
   n1 = Get SeasonVideoIDs!(self:Fort Runtime Options Object Reference=GetRuntimeOptions()) -> (SeasonVideoIDs:Season Video IDs Structure)
   n2 = Get VideoConfig() -> (VideoConfig:Video Widget Config Structure)
   > n3 = Branch(Condition="Equal Exactly (String)"(
     A=Break Season Video IDs!(SeasonVideoIDs:"Season Video IDs Structure (by ref)"=@n1) -> (Combo:String, EcosystemTrailer:String) {is=title K2Node_BreakStruct}
     B=Break Video Widget Config!(VideoWidgetConfig:"Video Widget Config Structure (by ref)"=@n2) -> (
      VideoUID:String
      bFadeFromColor:bool=~
      bEndFadeFromColor:bool=~
      bFadeUpToColor:bool=~
      bControlShaderPipelineCacheBatching:bool=~
      bTreatAudioOnlyAsColorFades:bool=~
      FadeColor:Linear Color Structure
      bPrePostRollNoAudioMix:bool=~
      bAllowGlobalForceGC:bool=~
      bInstanceForceGC:bool=~
     ) {is=title K2Node_BreakStruct}
    ))
   > Outputs!(True:Exec, False:Exec=@n3.else)
  ]}
]
variables [
 NewVar_0 {type=boolean}
 SecondsPriorToTransSound {type=number, default=2}
 TransitionSound {type=SoundBase, default=/Game/Sounds/Fort_Music/Seasonal/CineToBP_Transition_Cue}
]
widgets {
 info {inheritedWidgetCount=6, parentClass=/Script/FortniteUI.FortActivatableMovieWidget, rootWidgetClass=&SafeZone, widgetCount=10}
 widgets [
  MainSafeZone {
   widgetClassPath=&SafeZone
   bIsVariable=true
   widget=MainSafeZone
   properties {padBottom=false, padLeft=false, padRight=false, padTop=false}
  }
  SafeZonePadding {
   widgetClassPath=&Border
   bInherited=true
   bIsVariable=true
   parent=MainSafeZone
   widget=SafeZonePadding
   properties {brushColor=<rgba 0 0 0 0>, padding=<ltrb 0 0 0 0>}
   slot {class=SafeZoneSlot}
  }
  Overlay_0 {widgetClassPath=&Overlay, parent=SafeZonePadding, widget=Overlay_0, slot {class=BorderSlot, properties {padding=<ltrb 0 0 0 0>}}}
  ScaleBox {
   widgetClassPath=&ScaleBox
   parent=Overlay_0
   widget=ScaleBox
   slot {class=OverlaySlot, properties {horizontalAlignment=HAlign_Fill, verticalAlignment=VAlign_Fill}}
  }
  Image_VideoTexture {
   widgetClassPath=&Image
   bInherited=true
   bIsVariable=true
   parent=ScaleBox
   widget=Image_VideoTexture
   properties {brush {
     bIsDynamicallyLoaded=false
     drawAs=Image
     imageType=NoImage
     mirroring=NoMirror
     resourceName=None
     resourceObject=/Game/Movies/VideoPlayerWidgetMaterial_Unique
     tiling=NoTile
     imageSize=<1920 1080>
     margin=<ltrb 0 0 0 0>
     outlineSettings {
      bUseBrushTransparency=false
      roundingType=HalfHeightRadius
      width=0
      color {colorUseRule=UseColor_Specified, specifiedColor=<rgba 0 0 0 0>}
      cornerRadii=<0 0 0 0>
     }
     tintColor {colorUseRule=UseColor_Specified, specifiedColor=<rgba 1 1 1 1>}
     uVRegion {bIsValid=false, max=<0 0>, min=<0 0>}
    }}
   slot {class=ScaleBoxSlot}
  }
  Subtitles {
   widgetClassPath=/Script/SubtitlesWidgets.SubtitleDisplay
   bInherited=true
   bIsVariable=true
   parent=Overlay_0
   widget=Subtitles
   properties {
    options=/Game/UI/SubtitleOptions
    previewText=SEE-Bot: SEE-Bot SEES THE INSIDE OF THE BAG. THERE IS SO MUCH BAG TO SEE.
    visibility=HitTestInvisible
    wrapTextAt=1170
   }
   slot {class=OverlaySlot, properties {horizontalAlignment=HAlign_Center, verticalAlignment=VAlign_Bottom, padding=<ltrb 0 0 0 100>}}
  }
  SafeZone_0 {
   widgetClassPath=&SafeZone
   bIsVariable=true
   parent=Overlay_0
   widget=SafeZone_0
   properties {padLeft=false, padTop=false}
   slot {class=OverlaySlot, properties {horizontalAlignment=HAlign_Right, verticalAlignment=VAlign_Bottom, padding=<ltrb 0 0 20 20>}}
  }
  Button_Skip {
   widgetClassPath=/FortUILibrary/UIKit/Buttons/WBP_UIKit_Button_Quiet
   bInherited=true
   bIsVariable=true
   parent=SafeZone_0
   widget=Button_Skip
   properties {
    bHideInputActionWithKeyboard=true
    holdCompletedSound=/Game/Sounds/Fort_UI_Sounds/FrontEnd/New_FrontEnd/DefaultSounds/HoldableButtons/Small_Quiet_Clear/RenderedCine/UI_RenderedCine_H_oldable_Small_Complete_Cue_old
    holdPressedSound=/Game/Sounds/Fort_UI_Sounds/FrontEnd/New_FrontEnd/DefaultSounds/HoldableButtons/Small_Quiet_Clear/RenderedCine/UI_RenderedCine_H_oldable_Small_Pressed_Cue_old
    holdReleasedSound=/Game/Sounds/Fort_UI_Sounds/FrontEnd/New_FrontEnd/DefaultSounds/HoldableButtons/Small_Quiet_Clear/RenderedCine/UI_RenderedCine_H_oldable_Small_Release_Cue_old
    holdSound=/Game/Sounds/Fort_UI_Sounds/FrontEnd/New_FrontEnd/DefaultSounds/HoldableButtons/Small_Quiet_Clear/RenderedCine/UI_RenderedCine_H_oldable_Small_H_old_Cue_old
    text=Skip
    visibility=Collapsed
    hoveredSlateSoundOverride {resourceObject=/Game/Sounds/Fort_UI_Sounds/FrontEnd/SC_UI_Cinematics_SkipButton_Hover}
    pressedSlateSoundOverride {resourceObject=/Game/Sounds/Fort_UI_Sounds/FrontEnd/SC_UI_Cinematics_SkipButton_Pressed}
    triggeringInputAction {dataTable=&InputActionDataTable, rowName=GenericBack}
   }
   slot {class=SafeZoneSlot, properties {hAlign=HAlign_Left, vAlign=VAlign_Top}}
  }
  Image_Color {
   widgetClassPath=&Image
   bInherited=true
   bIsVariable=true
   parent=Overlay_0
   widget=Image_Color
   properties {visibility=HitTestInvisible}
   slot {class=OverlaySlot, properties {horizontalAlignment=HAlign_Fill, verticalAlignment=VAlign_Fill}}
  }
  Image_Throbber {
   widgetClassPath=&Image
   bInherited=true
   bIsVariable=true
   parent=Overlay_0
   widget=Image_Throbber
   properties {
    visibility=SelfHitTestInvisible
    brush {
     bIsDynamicallyLoaded=false
     drawAs=Image
     imageType=NoImage
     mirroring=NoMirror
     resourceName=None
     resourceObject=/Game/UI/Foundation/Materials/Throbber_Material-Monochrome
     tiling=NoTile
     imageSize=<64 64>
     margin=<ltrb 0 0 0 0>
     outlineSettings {
      bUseBrushTransparency=false
      roundingType=HalfHeightRadius
      width=0
      color {colorUseRule=UseColor_Specified, specifiedColor=<rgba 0 0 0 0>}
      cornerRadii=<0 0 0 0>
     }
     tintColor {colorUseRule=UseColor_Specified, specifiedColor=<rgba 1 1 1 1>}
     uVRegion {bIsValid=false, max=<0 0>, min=<0 0>}
    }
   }
   slot {class=OverlaySlot, properties {horizontalAlignment=HAlign_Center, verticalAlignment=VAlign_Center}}
  }
 ]
}
