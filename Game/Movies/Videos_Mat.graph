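# Videos_Mat is a User Interface domain material that, judging by its name and Movies folder, serves as a surface for video playback in UI.
# Its only node is a texture sample parameter, MediaPlayer_param, which defaults to PlaceholderMediaTexture.
# The exposed parameter lets material instances swap in another texture, such as a media player's output.
# The graph shows no link from the sample to a material output.
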Material /Game/Movies/Videos_Mat
graph {
 properties {materialDomain=MD_UI, pixelDepthOffsetMode=PDOM_Legacy, usedShadingModels=Default Lit}
 nodes [TextureSampleParameter2D() {parameterName=MediaPlayer_param, texture=&/PlaceholderMediaTexture}]
}
