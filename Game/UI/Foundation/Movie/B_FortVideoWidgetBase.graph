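# Base Fortnite video player widget: a scaled video image with subtitles, a loading throbber and an initially collapsed Skip button.
# Activation sets the cinematic control bus and media pre-end time and reports the view opened; deactivation clears the bus.
# The pre-end event plays a transition sound that persists across level loads, plus a leftover debug print.
# Skip logs analytics and broadcasts OnVideoHasBeenSkipped; helper functions override the video brush and size.
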WidgetBlueprint /Game/UI/Foundation/Movie/B_FortVideoWidgetBase
graphs [
 EventGraph {nodes [
   n1 = "On Clicked (Button_Skip)"!() -> (OutputDelegate, then, Button:Common Button Base Object Reference) {is=title K2Node_ComponentBoundEvent}
   > Fire Event Element Interacted(Element=@n1.Button)
   > Call On Video Has Been Skipped!(execute:Exec, self:B Fort Video Widget Base Object Reference) -> (then:Exec) {is=title K2Node_CallDelegate}
   n2 = PreEnded() {is=title K2Node_CustomEvent}
   > SpawnSound2D!(
    execute
    Sound:Sound Base Object Reference=Get TransitionSound() -> (TransitionSound:Sound Base Object Reference)
    VolumeMultiplier:float=1
    PitchMultiplier:float=1
    StartTime:float=0
    ConcurrencySettings:Sound Concurrency Object Reference
    bPersistAcrossLevelTransition:bool=true
    bAutoDestroy:bool=true
   ) -> (then, ReturnValue:Audio Component Object Reference)
   > PrintString!(execute, InString="Doing Work!", bPrintToScreen, bPrintToLog, TextColor, Duration, Key) -> (then)
   Event On Activated()
   > SetMediaPreEndTimeSeconds!(execute, self:Fort Video Widget Base Object Reference, InSeconds:float=Get SecondsPriorToTransSound() -> (SecondsPriorToTransSound:float)) -> (then)
   > Bind Event to Media Pre End Event(self:Fort Video Widget Base Object Reference, Delegate=@n2) {is=title K2Node_AddDelegate}
   > Fire Event Main View Opened!(execute, View:Widget Object Reference=Self-Reference(), AdditionalInfo:Flow Analytics Additional Info Structure, bIsOverlay:bool=false) -> (then)
   > Set Global Control Bus Mix Value!(execute, Bus:Sound Control Bus Object Reference=/Game/Sounds/Modulation/Buses/CB_Cinematic_IsActive, Value:float=1, FadeTime:float=0.25) -> (then) {is=title K2Node_CallFunction}
   Event On Deactivated()
   > Set Global Control Bus Mix Value(Bus=&CB_Cinematic_IsActive, Value=0, FadeTime=0.25)
   Event PreConstruct()
   Event Construct()
   Event Tick()
  ]}
 OnVideoHasBeenSkipped {nodes [OnVideoHasBeenSkipped() {is=title K2Node_FunctionEntry}]}
 OverrideImageVideoTextureBrush {nodes [
   n1 = OverrideImageVideoTextureBrush!() -> (then, Brush:Slate Brush Structure) {is=title K2Node_FunctionEntry}
   > SetBrush(self=Get Image_VideoTexture() -> (Image_VideoTexture:Image Object Reference), InBrush=@n1)
  ]}
 OverrideImageVideoTextureSize {nodes [
   n1 = OverrideImageVideoTextureSize!() -> (then, SizeX:double=~, SizeY:double=~) {is=title K2Node_FunctionEntry}
   > SetDesiredSizeOverride!(execute, self:Image Object Reference=Get Image_VideoTexture(), DesiredSize_X:double=@n1, DesiredSize_Y:double=@n1.SizeY) -> (then)
  ]}
]
variables [
 OnVideoHasBeenSkipped {}
 SecondsPriorToTransSound {type=number, default=2}
 TransitionSound {type=SoundBase, default=/Game/Sounds/Fort_Music/Seasonal/CineToBP_Transition_Cue}
]
widgets {
 info {inheritedWidgetCount=4, parentClass=/Script/FortniteUI.FortVideoWidgetBase, rootWidgetClass=&SafeZone, widgetCount=10}
 widgets [
  MainSafeZone {
   widgetClassPath=&SafeZone
   bIsVariable=true
   widget=MainSafeZone
   properties {padBottom=false, padLeft=false, padRight=false, padTop=false}
  }
  SafeZonePadding {
   widgetClassPath=&Border
   bIsVariable=true
   parent=MainSafeZone
   widget=SafeZonePadding
   properties {brushColor=<rgba 0 0 0 1>, padding=<ltrb 0 0 0 0>}
   slot {class=SafeZoneSlot}
  }
  Overlay_0 {widgetClassPath=&Overlay, parent=SafeZonePadding, widget=Overlay_0, slot {class=BorderSlot, properties {padding=<ltrb 0 0 0 0>}}}
  ScaleBox {
   widgetClassPath=&ScaleBox
   parent=Overlay_0
   widget=ScaleBox
   slot {class=OverlaySlot, properties {horizontalAlignment=HAlign_Fill, verticalAlignment=VAlign_Fill}}
  }
  Image_VideoTexture {
   widgetClassPath=&Image
   bInherited=true
   bIsVariable=true
   parent=ScaleBox
   widget=Image_VideoTexture
   properties {brush {
     bIsDynamicallyLoaded=false
     drawAs=Image
     imageType=NoImage
     mirroring=NoMirror
     resourceName=None
     resourceObject=/Game/Movies/VideoPlayerWidgetMaterial_Unique
     tiling=NoTile
     imageSize=<1920 1080>
     margin=<ltrb 0 0 0 0>
     outlineSettings {
      bUseBrushTransparency=false
      roundingType=HalfHeightRadius
      width=0
      color {colorUseRule=UseColor_Specified, specifiedColor=<rgba 0 0 0 0>}
      cornerRadii=<0 0 0 0>
     }
     tintColor {colorUseRule=UseColor_Specified, specifiedColor=<rgba 1 1 1 1>}
     uVRegion {bIsValid=false, max=<0 0>, min=<0 0>}
    }}
   slot {class=ScaleBoxSlot}
  }
  WBP_CaptureForPostBufferUpdate {
   widgetClassPath=&WBP_CaptureForPostBufferUpdate
   bIsVariable=true
   parent=Overlay_0
   slot {class=OverlaySlot, properties {horizontalAlignment=HAlign_Fill, verticalAlignment=VAlign_Fill}}
  }
  Subtitles {
   widgetClassPath=/Script/SubtitlesWidgets.SubtitleDisplay
   bInherited=true
   bIsVariable=true
   parent=Overlay_0
   widget=Subtitles
   properties {
    options=/Game/UI/SubtitleOptions
    previewText=SEE-Bot: SEE-Bot SEES THE INSIDE OF THE BAG. THERE IS SO MUCH BAG TO SEE.
    visibility=HitTestInvisible
    wrapTextAt=1170
   }
   slot {class=OverlaySlot, properties {horizontalAlignment=HAlign_Center, verticalAlignment=VAlign_Bottom, padding=<ltrb 0 0 0 100>}}
  }
  SafeZone_SkipButton {
   widgetClassPath=&SafeZone
   bIsVariable=true
   parent=Overlay_0
   widget=SafeZone_SkipButton
   properties {padLeft=false, padTop=false}
   slot {class=OverlaySlot, properties {horizontalAlignment=HAlign_Right, verticalAlignment=VAlign_Bottom, padding=<ltrb 0 0 32 32>}}
  }
  Button_Skip {
   widgetClassPath=/FortUILibrary/UIKit/Buttons/WBP_UIKit_Button_Quiet
   bInherited=true
   bIsVariable=true
   parent=SafeZone_SkipButton
   widget=Button_Skip
   properties {
    bHideInputActionWithKeyboard=true
    holdCompletedSound=/Game/Sounds/Fort_UI_Sounds/FrontEnd/New_FrontEnd/DefaultSounds/HoldableButtons/Small_Quiet_Clear/RenderedCine/UI_RenderedCine_H_oldable_Small_Complete_Cue_old
    holdPressedSound=/Game/Sounds/Fort_UI_Sounds/FrontEnd/New_FrontEnd/DefaultSounds/HoldableButtons/Small_Quiet_Clear/RenderedCine/UI_RenderedCine_H_oldable_Small_Pressed_Cue_old
    holdReleasedSound=/Game/Sounds/Fort_UI_Sounds/FrontEnd/New_FrontEnd/DefaultSounds/HoldableButtons/Small_Quiet_Clear/RenderedCine/UI_RenderedCine_H_oldable_Small_Release_Cue_old
    holdSound=/Game/Sounds/Fort_UI_Sounds/FrontEnd/New_FrontEnd/DefaultSounds/HoldableButtons/Small_Quiet_Clear/RenderedCine/UI_RenderedCine_H_oldable_Small_H_old_Cue_old
    text=Skip
    visibility=Collapsed
    hoveredSlateSoundOverride {resourceObject=/Game/Sounds/Fort_UI_Sounds/FrontEnd/SC_UI_Cinematics_SkipButton_Hover}
    pressedSlateSoundOverride {resourceObject=/Game/Sounds/Fort_UI_Sounds/FrontEnd/SC_UI_Cinematics_SkipButton_Pressed}
    triggeringInputAction {dataTable=&InputActionDataTable, rowName=GenericBack}
   }
   slot {class=SafeZoneSlot, properties {hAlign=HAlign_Left, vAlign=VAlign_Top}}
  }
  Image_Throbber {
   widgetClassPath=&Image
   bInherited=true
   bIsVariable=true
   parent=Overlay_0
   widget=Image_Throbber
   properties {
    visibility=SelfHitTestInvisible
    brush {
     bIsDynamicallyLoaded=false
     drawAs=Image
     imageType=FullColor
     mirroring=NoMirror
     resourceName=None
     resourceObject=/FortUILibrary/UIKit/Throbber/MI_UIKit_Monochrome_Throbber
     tiling=NoTile
     imageSize=<64 64>
     margin=<ltrb 0 0 0 0>
     outlineSettings {
      bUseBrushTransparency=false
      roundingType=HalfHeightRadius
      width=0
      color {colorUseRule=UseColor_Specified, specifiedColor=<rgba 0 0 0 0>}
      cornerRadii=<0 0 0 0>
     }
     tintColor {colorUseRule=UseColor_Specified, specifiedColor=<rgba 1 1 1 1>}
     uVRegion {bIsValid=false, max=<0 0>, min=<0 0>}
    }
   }
   slot {class=OverlaySlot, properties {horizontalAlignment=HAlign_Center, verticalAlignment=VAlign_Center}}
  }
 ]
}
